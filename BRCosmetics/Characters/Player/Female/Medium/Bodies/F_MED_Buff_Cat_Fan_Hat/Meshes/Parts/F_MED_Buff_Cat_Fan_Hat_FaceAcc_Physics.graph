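# Physics asset for the face accessory mesh of the medium female Buff Cat Fan Hat cosmetic, covering its dynamic hair bones.
# Capsule bodies cover the neck, head, hair strands, front part, clips and two buns.
# Strands and buns constrain to a master hair body on the head, with back hair routed through a secondary master.
# The neck, head and master links swing and twist freely, while hair pieces have tight twist and varied swing limits.
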PhysicsAsset /BRCosmetics/Characters/Player/Female/Medium/Bodies/F_MED_Buff_Cat_Fan_Hat/Meshes/Parts/F_MED_Buff_Cat_Fan_Hat_FaceAcc_Physics
bodies {
 Dyn_Back_Hair1 [{shapeType=Capsule, length=1, radius=1, shapeName=Dyn_Back_Hair1_capsule, center=<1 0 0>, rotation=<rot 90 0 0>}]
 Dyn_Back_Hair2 [{shapeType=Capsule, length=1, radius=1, shapeName=Dyn_Back_Hair2_capsule, center=<1 0 0>, rotation=<rot 90 0 0>}]
 Dyn_Back_Hair3 [{shapeType=Capsule, length=1, radius=1, shapeName=Dyn_Back_Hair3_capsule, center=<1 0 0>, rotation=<rot 90 0 0>}]
 Dyn_Backstrand_l [{shapeType=Capsule, length=2, radius=1, shapeName=Dyn_Backstrand_l_capsule, center=<1 0 0>, rotation=<rot 90 0 0>}]
 Dyn_Backstrand_r [{shapeType=Capsule, length=1, radius=1, shapeName=Dyn_Backstrand_r_capsule, center=<0.1 0 0>, rotation=<rot 90 0 0>}]
 Dyn_EarStrand_l [{shapeType=Capsule, length=1, radius=0.7, shapeName=Dyn_EarStrand_l_capsule, center=<0.1 0 0>, rotation=<rot 90 0 0>}]
 Dyn_EarStrand_l1 [{shapeType=Capsule, length=3, radius=1, shapeName=Dyn_EarStrand_l1_capsule, center=<1 0 0>, rotation=<rot 90 0 0>}]
 Dyn_EarStrand_l2 [{shapeType=Capsule, length=3, radius=1, shapeName=Dyn_EarStrand_l2_capsule, center=<0.1 0 0>, rotation=<rot 90 0 0>}]
 Dyn_EarStrand_r [{shapeType=Capsule, length=5, radius=0.7, shapeName=Dyn_EarStrand_r_capsule, center=<1.2412 0 -0.0001>, rotation=<rot 90 0 0>}]
 Dyn_Front_Part [{shapeType=Capsule, length=1, radius=1, shapeName=Dyn_Front_Part_capsule, center=<0.1 0 0>, rotation=<rot 90 0 0>}]
 Dyn_HairPart_Clips [{shapeType=Capsule, length=1, radius=1, shapeName=Dyn_HairPart_Clips_capsule, center=<0.1 0 0>, rotation=<rot 90 0 0>}]
 Dyn_Left_Bun [{shapeType=Capsule, length=3, radius=3, shapeName=Dyn_Left_Bun_capsule, center=<-2.0715 -2.952 1.8737>, rotation=<rot -23.518 -90 -32.4666>}]
 Dyn_MasterBack_Hair [{shapeType=Capsule, length=0.505, radius=0.505, shapeName=Dyn_MasterBack_Hair_capsule, rotation=<rot 0 -90 0>}]
 Dyn_Master_Hair [{shapeType=Capsule, length=0.5, radius=0.5, shapeName=Dyn_Master_Hair_capsule, rotation=<rot 0 -90 0>}]
 Dyn_Right_Bun [{shapeType=Capsule, length=3, radius=3, shapeName=Dyn_Right_Bun_capsule, center=<-1.1841 -4.3032 -1.097>, rotation=<rot 40.8352 -90 30.6397>}]
 head [{shapeType=Capsule, length=8, radius=7, shapeName=head_capsule, center=<6.05 0.288 -0.0001>, rotation=<rot 90 0 0>}]
 neck_01 [{shapeType=Capsule, length=20, radius=5, shapeName=neck_01_capsule, center=<-0.007 -0.6142 0>, rotation=<rot 0 -90 0>}]
 neck_02 [{shapeType=Capsule, length=10, radius=6, shapeName=neck_02_capsule, center=<0 2.044 0>, rotation=<rot 90 0 0>}]
}
constraints [
 {bone1Name=Dyn_Front_Part, bone2Name=Dyn_Master_Hair, swing1LimitDegrees=1, swing2LimitDegrees=20, twistLimitDegrees=1}
 {bone1Name=Dyn_EarStrand_r, bone2Name=Dyn_Master_Hair, swing1LimitDegrees=10, swing2LimitDegrees=30, twistLimitDegrees=1}
 {bone1Name=Dyn_EarStrand_l, bone2Name=Dyn_Master_Hair, swing1LimitDegrees=45, swing2LimitDegrees=20, twistLimitDegrees=1}
 {bone1Name=Dyn_EarStrand_l2, bone2Name=Dyn_Master_Hair, swing1LimitDegrees=45, swing2LimitDegrees=20, twistLimitDegrees=1}
 {bone1Name=Dyn_HairPart_Clips, bone2Name=Dyn_Master_Hair, swing1LimitDegrees=45, swing2LimitDegrees=20, twistLimitDegrees=1}
 {bone1Name=Dyn_Backstrand_r, bone2Name=Dyn_Master_Hair, swing1LimitDegrees=45, swing2LimitDegrees=20, twistLimitDegrees=1}
 {bone1Name=Dyn_Backstrand_l, bone2Name=Dyn_Master_Hair, swing1LimitDegrees=45, swing2LimitDegrees=20, twistLimitDegrees=1}
 {bone1Name=Dyn_EarStrand_l1, bone2Name=Dyn_Master_Hair, swing1LimitDegrees=45, swing2LimitDegrees=20, twistLimitDegrees=1}
 {bone1Name=Dyn_Back_Hair1, bone2Name=Dyn_MasterBack_Hair, swing1LimitDegrees=45, swing2LimitDegrees=20, twistLimitDegrees=1}
 {bone1Name=Dyn_Back_Hair2, bone2Name=Dyn_MasterBack_Hair, swing1LimitDegrees=45, swing2LimitDegrees=20, twistLimitDegrees=1}
 {bone1Name=Dyn_Back_Hair3, bone2Name=Dyn_MasterBack_Hair, swing1LimitDegrees=45, swing2LimitDegrees=20, twistLimitDegrees=1}
 {
  bone1Name=Dyn_MasterBack_Hair
  bone2Name=Dyn_Master_Hair
  swing1LimitDegrees=45
  swing1Motion=Free
  swing2LimitDegrees=45
  swing2Motion=Free
  twistLimitDegrees=45
  twistMotion=Free
 }
 {
  bone1Name=Dyn_Master_Hair
  bone2Name=head
  swing1LimitDegrees=45
  swing1Motion=Free
  swing2LimitDegrees=45
  swing2Motion=Free
  twistLimitDegrees=45
  twistMotion=Free
 }
 {
  bone1Name=head
  bone2Name=neck_02
  swing1LimitDegrees=45
  swing1Motion=Free
  swing2LimitDegrees=45
  swing2Motion=Free
  twistLimitDegrees=45
  twistMotion=Free
 }
 {bone1Name=Dyn_Right_Bun, bone2Name=Dyn_Master_Hair, swing1LimitDegrees=1, swing2LimitDegrees=10, twistLimitDegrees=1}
 {bone1Name=Dyn_Left_Bun, bone2Name=Dyn_Master_Hair, swing1LimitDegrees=1, swing2LimitDegrees=10, twistLimitDegrees=1}
 {
  bone1Name=neck_02
  bone2Name=neck_01
  swing1LimitDegrees=45
  swing1Motion=Free
  swing2LimitDegrees=45
  swing2Motion=Free
  twistLimitDegrees=45
  twistMotion=Free
 }
]
