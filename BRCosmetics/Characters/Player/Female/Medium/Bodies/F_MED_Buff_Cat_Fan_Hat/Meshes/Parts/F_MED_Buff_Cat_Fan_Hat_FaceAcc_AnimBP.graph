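# Animation Blueprint for the face accessory of the F_MED_Buff_Cat_Fan_Hat character, adding pose correction and physics to its dynamic bones.
# It copies the pose from the attached parent mesh, then additively yaws Dyn_MasterBack_Hair, weighted by Head_Up.
# The event graph derives Head_Up each update from the head bone's rotation delta relative to neck_01.
# A RigidBody node then simulates in neck_01 space with overridden gravity and damped component motion.
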AnimBlueprint /BRCosmetics/Characters/Player/Female/Medium/Bodies/F_MED_Buff_Cat_Fan_Hat/Meshes/Parts/F_MED_Buff_Cat_Fan_Hat_FaceAcc_AnimBP
target_skeleton=&/F_MED_Buff_Cat_Fan_Hat_FaceAcc_Skeleton
graphs [
 AnimGraph {nodes [
   n1 = "Transform (Modify) Bone - Bone: Dyn_MasterBack_Hair"(
    ComponentPose=Local To Component(LocalPose=Copy Pose From Mesh() {bUseAttachedParent=true})
    Alpha=Get Head_Up() -> (Head_Up:double)
    Translation="0,0,0"
    Rotation="0,-40,0"
    Scale="1,1,1"
   ) {
    rotationMode=BMM_Additive
    rotationSpace=BCS_BoneSpace
    translationMode=BMM_Additive
    translationSpace=BCS_BoneSpace
    alphaBoolBlend {blendOption=Linear}
    boneToModify {boneName=Dyn_MasterBack_Hair}
    rotation {yaw=-40}
   }
   n2 = RigidBody(ComponentPose=@n1) {
    bClampLinearTranslationLimitToRefPose=true
    bOverrideWorldGravity=true
    overridePhysicsAsset=&/F_MED_Buff_Cat_Fan_Hat_FaceAcc_Physics
    simulationSpace=BaseBoneSpace
    alphaBoolBlend {blendOption=Linear}
    baseBoneRef {boneName=neck_01}
    componentLinearAccScale=<0.1 0.5 0.3>
    componentLinearVelScale=<0.1 0.5 0.3>
    overrideWorldGravity {z=-980}
   }
   Output Pose(Result=Component To Local(ComponentPose=@n2)) {layerGroup=None}
  ]}
 EventGraph {nodes [
   Event BlueprintUpdateAnimation()
   > Set Head_Up(Head_Up:double=GetAlphaFromReferenceTransformDelta(SkeletalMeshComponent=Get SkelMesh(), BoneName=head, BaseName=neck_01, DeltaComponentType=EulerY, InputRange=(X=-1,Y=12), TwistAxis="1,0,0")) -> (Output_Get:double)
   Event BlueprintInitializeAnimation()
   > Set SkelMesh(SkelMesh=GetOwnerSkeletalMeshComponent())
  ]}
]
variables [
 Head_Up {type=number}
 SkelMesh {type=SkeletalMeshComponent}
]
